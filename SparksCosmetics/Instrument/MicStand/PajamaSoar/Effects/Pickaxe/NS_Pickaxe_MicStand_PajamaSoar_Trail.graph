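# Niagara trail effect for the PajamaSoar mic stand pickaxe, drawn as two short-lived ribbons from anim-trail modules following mesh bones or sockets.
# Hard_Lines is a camera-offset, blue-tinted ribbon that barely narrows, while the wider Trail_Instrument_Color ribbon tapers to zero.
# Both fade over life, with alpha adjusted for front end and time of day.
# User.IdleFXOn switches alpha strength on both and the color gradient on Hard_Lines.
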NiagaraSystem /SparksCosmetics/Instrument/MicStand/PajamaSoar/Effects/Pickaxe/NS_Pickaxe_MicStand_PajamaSoar_Trail
dependencies {
 usedDataInterfaces [&NiagaraDataInterfaceCurve, &NiagaraDataInterfaceColorCurve]
 usedDynamicInputs []
 usedModules [
  &SystemState
  &self:SystemSpawnScript.SystemScriptSource.SystemScriptGraph.NiagaraNodeAssignment_0.SetVariables_#1
 ]
}
emitters {
 Hard_Lines {
  emitterUpdateScript [
   &EmitterState(Life Cycle Mode=::NewEnumerator1, Loop Behavior=::NewEnumerator1, Loop Duration=0.22, UseLoopDelay=true)
   &FN_SpawnAnimTrail_v2(AnimTrailWidthMode=::FromFirst)
  ]
  particleSpawnScript [
   /Game/Effects/Niagara/Modules/AnimTrail/FN_InitializeAnimTrail/FN_InitializeAnimTrailBoneIndecies(First Bone or Socket Index=1)
   /Game/Effects/Niagara/Modules/AnimTrail/FN_InitializeAnimTrail/FN_InitializeAnimTrail(AnimTrailWidth=0.85, Color=<rgba 1 1 1 0.7>, Lifetime=0.2)
   &SystemLocation(MaskSpawnGroup=false)
  ]
  particleUpdateScript [
   &CameraOffset(Camera Offset Amount=-25)
   &ScaleRibbonWidth(Ribbon Width Scale=&FloatFromCurve(FloatCurve=DataInterface() {propertyValues {Curve {keys [
         {arriveTangent=-1.2579, interpMode=RCIM_Cubic, time=0, value=1}
         {arriveTangent=-0.3747, interpMode=RCIM_Cubic, leaveTangent=-0.3747, tangentMode=RCTM_User, time=1.0133, value=0.8}
        ]}}}))
   &ScaleColor(
    Scale Alpha=&Multiply_Float(
     A=&FloatFromCurve(
      FloatCurve=DataInterface() {propertyValues {Curve {keys [
          {interpMode=RCIM_Cubic, leaveTangent=-2.3646, tangentMode=RCTM_Break, time=0, value=1}
          <at 0.6689 0.0233>
          {arriveTangent=-0.2139, interpMode=RCIM_Cubic, tangentMode=RCTM_Break, time=1, value=0}
         ]}}}
      Scale Curve=&MakeCustomFloatFromBool(
       Bool=&SetBoolByFloatComparison(A=$User.IdleFXOn, B=0)
       True Float=&FN_FrontEndFloatModifier(Gameplay=1, Front End=0.3)
       False Float=&FN_FrontEndFloatModifier(Gameplay=0.7, Front End=0.2)
      )
     )
     B=&FNTimeOfDayScalar(Night Scale=0.5, Morning Scale=0.65, Day Scale=0.7, Evening Scale=0.7)
    )
    Scale RGB=&MakeCustomVectorFromBool(
     Bool=&SetBoolByFloatComparison(A=$User.IdleFXOn, B=0)
     True Vector=&Multiply_VectorByFloat(
      Vector=&Multiply_Vector(
       A=&MakeVectorFromLinearColorRGB(Color=&ColorFromCurve(ColorCurve=DataInterface() {propertyValues {
           AlphaCurve {keys [<at 0 1 RCIM_Cubic>,<at 1 0 RCIM_Cubic>]}
           BlueCurve {keys [<at 0 0.2552>,<at 0.9883 0.12>]}
           GreenCurve {keys [<at 0 0.1389>,<at 0.9883 0.0583>]}
           RedCurve {keys [<at 0 0.0279>,<at 0.9883 0.0156>]}
          }}))
       B=&MakeVectorFromLinearColorRGB(Color=$System.Color TOD)
      )
      Float=&FN_FrontEndFloatModifier(Gameplay=1, Front End=0.35)
     )
     False Vector=&MakeVectorFromLinearColorRGB(Color=&ColorFromCurve(ColorCurve=DataInterface() {propertyValues {
         AlphaCurve {keys [<at 0 1 RCIM_Cubic>]}
         BlueCurve {keys [<at 0.056 0.0729>,<at 0.9928 0.04>]}
         GreenCurve {keys [<at 0.056 0.0502>,<at 0.9928 0.0261>]}
         RedCurve {keys [<at 0.056 0.0239>,<at 0.9928 0.0094>]}
        }}))
    )
    ScaleA=true
    ScaleRGB=true
   ) {module=ScaleColor002}
   &FN_UpdateAnimTrail_v2()
  ]
  rendererClasses [&NiagaraRibbonRendererProperties]
  renderers [{rendererClass=&NiagaraRibbonRendererProperties}]
 }
 Trail_Instrument_Color {
  emitterUpdateScript [
   &EmitterState(Life Cycle Mode=::NewEnumerator1, Loop Behavior=::NewEnumerator1, Loop Duration=0.22, UseLoopDelay=true)
   &FN_SpawnAnimTrail_v2()
  ]
  particleSpawnScript [
   &FN_InitializeAnimTrailBoneIndecies(First Bone or Socket Index=1)
   &FN_InitializeAnimTrail(AnimTrailWidth=1.2, Lifetime=0.25)
   &SystemLocation(MaskSpawnGroup=false)
  ]
  particleUpdateScript [
   &ScaleRibbonWidth(Ribbon Width Scale=&FloatFromCurve(FloatCurve=DataInterface() {propertyValues {Curve {keys [
         {arriveTangent=-1.2579, interpMode=RCIM_Cubic, time=0, value=1}
         {arriveTangent=-0.9664, interpMode=RCIM_Cubic, leaveTangent=-0.9664, tangentMode=RCTM_User, time=1, value=0}
        ]}}}))
   &ScaleColor(
    Scale Alpha=&Multiply_Float(
     A=&FloatFromCurve(
      FloatCurve=DataInterface() {propertyValues {Curve {keys [
          {arriveTangent=-0.7187, interpMode=RCIM_Cubic, leaveTangent=-1.2793, tangentMode=RCTM_Break, time=0, value=1}
          {arriveTangent=-0.6445, interpMode=RCIM_Cubic, leaveTangent=-0.6445, tangentMode=RCTM_User, time=0.4966, value=0.3176}
          {arriveTangentWeight=0.1962, interpMode=RCIM_Cubic, tangentMode=RCTM_Break, time=1, value=0}
         ]}}}
      Scale Curve=&MakeCustomFloatFromBool(
       Bool=&SetBoolByFloatComparison(A=$User.IdleFXOn, B=0)
       True Float=&FN_FrontEndFloatModifier(Gameplay=1, Front End=0.5)
       False Float=&FN_FrontEndFloatModifier(Gameplay=0.9, Front End=0.5)
      )
     )
     B=&FNTimeOfDayScalar(Night Scale=0.5, Morning Scale=0.65, Day Scale=0.7, Evening Scale=0.7)
    )
    Scale RGB=&Multiply_Vector(A=<1 1 1>, B=&MakeVectorFromLinearColorRGB(Color=$System.Color TOD))
    ScaleA=true
    ScaleRGB=true
   ) {module=ScaleColor002}
   &FN_UpdateAnimTrail_v2()
  ]
  rendererClasses [&NiagaraRibbonRendererProperties]
  renderers [{rendererClass=&NiagaraRibbonRendererProperties}]
 }
}
summary {
 systemName=NS_Pickaxe_MicStand_PajamaSoar_Trail
 emitters [
  {emitterName=Trail_Instrument_Color, rendererClasses [&NiagaraRibbonRendererProperties]}
  {emitterName=Hard_Lines, rendererClasses [&NiagaraRibbonRendererProperties]}
 ]
 userVariables [
  User.bIsFrontEnd {type=NiagaraBool}
  User.bIsFrontEndPreview {type=NiagaraBool}
  User.IdleFXOn {type=NiagaraFloat, defaultValue=1, description=""}
  User.RibbonIDOffset {type=NiagaraInt32}
  User.SkeletalMesh {
   type=NiagaraDataInterfaceSkeletalMesh
   defaultValue {
    struct=NiagaraExt_VariableValue_DataInterface
    value {dataInterface=&self:NiagaraDataInterfaceSkeletalMesh_0}
   }
  }
  User.TrailColor {type=LinearColor, description="", defaultValue {struct=LinearColor, value=<rgba 1 1 1 0.25>}}
  User.TrailSecondaryColor {type=LinearColor, description="", defaultValue {struct=LinearColor, value=<rgba 10 10 10 0.25>}}
 ]
}
